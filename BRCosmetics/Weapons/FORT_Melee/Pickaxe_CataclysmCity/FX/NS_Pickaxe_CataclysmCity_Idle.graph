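# Niagara system for the idle effect of the Cataclysm City pickaxe, with left and right glow emitters.
# GlowL and GlowR are enabled but list no modules or renderers here.
# The disabled Glows emitter bursts two looping purple sprites, mirrored along X by partition index and nudged by camera offset.
# User parameters expose a front-end preview flag and a skeletal mesh, though the Glows modules shown read neither.
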NiagaraSystem /BRCosmetics/Weapons/FORT_Melee/Pickaxe_CataclysmCity/FX/NS_Pickaxe_CataclysmCity_Idle
dependencies {
 usedDataInterfaces []
 usedDynamicInputs []
 usedModules [
  &self:SystemSpawnScript.SystemScriptSource.SystemScriptGraph.NiagaraNodeAssignment_1.SetVariables_#1
  &self:SystemSpawnScript.SystemScriptSource.SystemScriptGraph.NiagaraNodeAssignment_0.SetVariables_#2
  &SystemState
 ]
}
emitters {
 GlowL {rendererClasses [], renderers []}
 GlowR {rendererClasses [], renderers []}
 Glows {
  bEnabled=false
  emitterUpdateScript [
   &EmitterState(Life Cycle Mode=::NewEnumerator1, Loop Behavior=::NewEnumerator1, Loop Duration=1, UseLoopDelay=false)
   &SpawnBurst_Instantaneous(Spawn Count=2, Use Spawn Probability=false, UseLoopCountLimit=false)
  ]
  particleSpawnScript [
   &PartitionParticles()
   &InitializeParticle(
    Color=<rgba 0.9822 0 2 0.3>
    Position Offset=&Multiply_VectorByFloat(
     Vector=<2.5 0 0>
     Float=&MergeHelper_FloatSwitch(0=1, 1=-1, 2=?, 3=?, Index=$Particles.PartitionParticles.Partition, Num=::NewEnumerator0)
    )
    Sprite Size Mode=::NewEnumerator3
    Uniform Sprite Size=19.0813
    UsePositionOffset=true
    Write Material Random=true
   )
  ]
  particleUpdateScript [
   &ParticleState(Kill Particles When Lifetime Has Elapsed=false, Loop Particles Lifetime=true)
   &CameraOffset(Camera Offset Amount=5)
  ]
  rendererClasses [&NiagaraSpriteRendererProperties]
  renderers [{rendererClass=&NiagaraSpriteRendererProperties}]
 }
}
summary {
 systemName=NS_Pickaxe_CataclysmCity_Idle
 emitters [
  {bEnabled=false, emitterName=Glows, rendererClasses [&NiagaraSpriteRendererProperties]}
  {emitterName=GlowL, rendererClasses []}
  {emitterName=GlowR, rendererClasses []}
 ]
 userVariables [
  User.bIsFrontEndPreview {type=NiagaraBool, defaultValue=true}
  User.SkeletalMesh {
   type=NiagaraDataInterfaceSkeletalMesh
   defaultValue {
    struct=NiagaraExt_VariableValue_DataInterface
    value {dataInterface=&self:NiagaraDataInterfaceSkeletalMesh_1}
   }
  }
 ]
}
